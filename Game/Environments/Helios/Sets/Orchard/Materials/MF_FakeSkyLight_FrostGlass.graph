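# Material function that overlays a fake sky-lit frosted greenhouse glass look, using Two Sided Foliage shading, onto incoming material attributes.
# Color combines a tinted sky-light cubemap sample, noise-based fog and Fresnel, with time-of-day variation, feeding base, subsurface and exposure-compensated emissive, optionally night-only.
# An SFX mask channel, optionally limited by UV1 layer index, drives the attribute blend.
# ShadowReplace offsets shadow-pass vertices toward the object pivot, apparently collapsing glass shadows.
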MaterialFunction /Game/Environments/Helios/Sets/Orchard/Materials/MF_FakeSkyLight_FrostGlass
graph {nodes [
  n1 = TextureCoordinate()
  ComponentMask(None=@n1) {g=true}
  n2 = FunctionInput() {inputType=FunctionInput_MaterialAttributes}
  n3 = GetMaterialAttributes(None=@n2) {attributeGetTypes [#1, #2]}
  n4 = SetMaterialAttributes!(
   MaterialAttributes=@n3.MaterialAttributes
   BaseColor=NamedRerouteUsage()
   SubsurfaceColor=NamedRerouteUsage()
   ShadingModel=ShadingModel() {shadingModel=MSM_TwoSidedFoliage}
   EmissiveColor=StaticSwitchParameter(True=Constant3Vector() {constant=<rgba 3 3 3 0>}, False=NamedRerouteUsage()) {group=GreenHouse, parameterName=Debuggor}
   Opacity=ScalarParameter() {defaultValue=0.1, group=GreenHouse, parameterName=SSSOpacity}
   WorldPositionOffset=NamedRerouteUsage()
  ) {attributeSetTypes [
    #1
    #3
    #4
    #5
    #6
    #2
   ]}
  n5 = TextureSampleParameter2D() {
   group=GreenHouse
   parameterName=SFX_Mask
   samplerType=SAMPLERTYPE_LinearGrayscale
   texture=/Game/Global/Base4Layers_2022/Textures/T_SFX_Grayscale
  }
  n6 = NamedRerouteDeclaration(None=ChannelMaskParameter(None=@n5.RGB) {group=GreenHouse, parameterName=SkyLightChannel}.R) {name=SFXMask, nodeColor=<rgba 1 0.4219 0 1>}
  n7 = ComponentMask(None=TextureCoordinate() {coordinateIndex=1}) {r=true}
  n8 = ScalarParameter() {defaultValue=1, group=GreenHouse, parameterName=LayerIndexR}
  n9 = Constant()
  n10 = Constant()
  n11 = Multiply(
   A=If(A=@n7, B=@n8, "A > B"=Constant() {r=1}, "A == B"=@n9, "A < B"=@n9) {constB=1}
   B=If(A=@n7, B=Add(A=@n8), "A > B"=@n10, "A == B"=@n10, "A < B"=Constant() {r=1}) {constB=1}
  )
  n12 = ComponentMask(None=TextureCoordinate() {coordinateIndex=1}) {r=true}
  n13 = ScalarParameter() {defaultValue=1, group=GreenHouse, parameterName=AdditionalLayerIndexR}
  n14 = Constant() {r=1}
  n15 = Constant()
  n16 = Add(A=@n13)
  n17 = Constant()
  n18 = Constant() {r=1}
  n19 = Add(
   A=@n11
   B=Multiply(
    A=If(A=@n12, B=@n13, "A > B"=@n14, "A == B"=@n15, "A < B"=@n15) {constB=1}
    B=If(A=@n12, B=@n16, "A > B"=@n17, "A == B"=@n17, "A < B"=@n18) {constB=1}
   )
  )
  n20 = StaticSwitchParameter(True=Multiply(A=@n6, B=Saturate(None=@n19)), False=Multiply(A=@n6, B=@n11)) {group=GreenHouse, parameterName=UseAdditionalLayers}
  FunctionOutput(None=BlendMaterialAttributes(A=@n2, B=@n4, Alpha=StaticSwitchParameter(True=@n20, False=@n6) {group=GreenHouse, parameterName=Use4Layers}))
  n21 = TextureCoordinate()
  n22 = ScalarParameter() {defaultValue=0.003, group=GreenHouse, parameterName=NormalNoiseTillingWS}
  Multiply(A=@n21, B=@n22)
  n23 = ComponentMask(None=@n1) {r=true}
  Multiply(A=Abs(None=Subtract(A=@n23) {constB=0.5})) {constB=2}
  n24 = Multiply(A=ComponentMask(None=WorldPosition().XYZ) {g=true, r=true}, B=@n22) {constB=0.005}
  n25 = Multiply(A=ComponentMask(None=TextureSample(UVs=@n24) {samplerType=SAMPLERTYPE_Normal, texture=/Game/Global/Textures/Default/Noise/T_SwirlNoise_N}.RGB) {g=true, r=true}) {constB=0.1}
  n26 = Multiply(A=TextureSample(UVs=Add(A=@n21, B=@n25)) {samplerType=SAMPLERTYPE_LinearGrayscale, texture=/Game/Global/Textures/Default/Noise/LowResBlurredNoise_NoMips}.R) {constB=0.2}
  n27 = TextureObjectParameter() {
   group=GreenHouse
   parameterName=FoggyNoise
   samplerType=SAMPLERTYPE_Masks
   texture=/Game/Environments/Helios/Sets/Estate/Textures/T_FoggyNoise_Mask_TechArt
  }
  n28 = TextureSample(UVs=Multiply(A=@n21, B=ScalarParameter() {defaultValue=3, group=GreenHouse, parameterName=LargeNoiseTilling}) {constB=2}, Tex=@n27) {
   samplerSource=SSM_Wrap_WorldGroupSettings
   samplerType=SAMPLERTYPE_Masks
   texture=&T_FoggyNoise_Mask_TechArt
  }
  n29 = ScalarParameter() {defaultValue=5, group=GreenHouse, parameterName=NoiseTilling}
  n30 = TextureSample(UVs=Add(A=Multiply(A=@n21, B=@n29) {constB=3}, B=@n25), Tex=@n27) {
   samplerSource=SSM_Wrap_WorldGroupSettings
   samplerType=SAMPLERTYPE_Masks
   texture=&T_FoggyNoise_Mask_TechArt
  }
  n31 = Multiply(A=Multiply(A=Power(Base=@n28.B) {constExponent=3}, B=@n30.G)) {constB=0.6}
  n32 = VertexNormalWS()
  n33 = Saturate(None=Abs(None=ComponentMask(None=@n32) {b=true}))
  n34 = LinearInterpolate(Alpha=SmoothStep(Value=@n33) {constMax=0.2}) {constA=0.3, constAlpha=1, constB=0.2}
  n35 = StaticSwitchParameter(True=@n26, False=LinearInterpolate(A=Add(A=@n31, B=@n34), B=@n26, Alpha=SmoothStep(Value=@n33) {constMax=0.8})) {defaultValue=true, group=GreenHouse, parameterName=UseFogOnly}
  Multiply(A=@n35) {constB=3}
  n36 = VectorParameter() {group=GreenHouse, parameterName=SkyLightTint, defaultValue=<rgba 1 1 1 1>}
  n37 = CameraVectorWS()
  n38 = RotateAboutAxis(
   NormalizedRotationAxis=Constant3Vector() {constant=<rgba 0 0 1 0>}
   RotationAngle=Divide(A=Constant() {r=180}) {constB=360}
   PivotPoint=Constant()
   Position=@n37
  )
  n39 = MaterialFunctionCall(Float3=Normalize(VectorInput=Add(A=@n38, B=@n37))) {materialFunction=&BreakOutFloat3Components}
  n40 = SkyLightEnvMapSample(
   Direction=MaterialFunctionCall(X=@n39.R, Y=@n39.R, Z=Multiply(A=@n39.B) {constB=-1}) {materialFunction=&MakeFloat3}.Result
   Roughness=ScalarParameter() {defaultValue=0.22, group=GreenHouse, parameterName=CubeMapRoughness}
  )
  n41 = ScalarParameter() {defaultValue=0.5, group=GreenHouse, parameterName=CubeMapColorIntensity}
  n42 = Saturate(None=ScalarParameter() {defaultValue=0.8, group=GreenHouse, parameterName=CubeMapColorIntensity_Night})
  n43 = Multiply(
   A=Clamp(None=Multiply(A=@n36.RGB, B=@n40), Max=ScalarParameter() {defaultValue=0.7, group=GreenHouse, parameterName=CubeMapMaxBrightness})
   B=MaterialFunctionCall(Value_Morning=@n41, Value_Day=@n41, Value_Evening=@n42, Value_Night=@n42) {materialFunction=&Scalar_Time_of_Day}.Result
  )
  n44 = VectorParameter() {group=GreenHouse, parameterName=SolidColor, defaultValue=<rgba 0.6025 0.6927 0.6908 1>}
  n45 = ScalarParameter() {group=GreenHouse, parameterName=SolidColorBlendSky}
  NamedRerouteDeclaration(None=Desaturation(None=LinearInterpolate(A=@n43, B=@n44.RGB, Alpha=@n45), Fraction=ScalarParameter() {group=GreenHouse, parameterName=SkyLightDesaturation})) {name=SkyLightColor, nodeColor=<rgba 0 1 0.3438 1>}
  n46 = NamedRerouteUsage()
  n47 = NamedRerouteUsage()
  n48 = NamedRerouteUsage()
  n49 = Multiply(A=Fresnel(), B=VectorParameter() {group=GreenHouse, parameterName=FresnelColor, defaultValue=<rgba 0.6979 1 0.9937 1>}.RGB)
  n50 = ScalarParameter() {defaultValue=0.5, group=GreenHouse, parameterName=FresnelIntensity}
  n51 = Divide(A=@n50)
  n52 = MaterialFunctionCall(Color_Morning=@n50, Color_Day=@n50, Color_Evening=@n51, Color_Night=@n51) {materialFunction=&Color_Time_of_Day}
  n53 = Saturate(None=Add(A=LinearInterpolate(A=@n46, B=@n47, Alpha=@n48), B=Multiply(A=@n49, B=@n52.Result)))
  n54 = Constant()
  n55 = ScalarParameter() {defaultValue=0.1, group=GreenHouse, parameterName=EmissiveIntensity}
  n56 = Multiply(
   A=Divide(A=@n53, B=EyeAdaptation().EyeAdaptation)
   B=StaticSwitchParameter(
    True=MaterialFunctionCall(Value_Morning=@n54, Value_Day=@n54, Value_Evening=@n54, Value_Night=@n55) {materialFunction=&Scalar_Time_of_Day}.Result
    False=@n55
   ) {defaultValue=true, group=GreenHouse, parameterName=OnlyLightInNight}
  ) {constB=0.2}
  n57 = ScalarParameter() {defaultValue=0.5, group=GreenHouse, parameterName=EmissiveDesaturation}
  NamedRerouteDeclaration(None=Multiply(
    A=Desaturation(None=@n56, Fraction=@n57)
    B=VectorParameter() {group=GreenHouse, parameterName=EmissiveColorTint, defaultValue=<rgba 1 1 1 1>}.RGB
   )) {name=EmissiveOUT, nodeColor=<rgba 0.875 0 1 1>}
  n58 = TextureSampleParameter2D() {group=GreenHouse, parameterName=GreenHouseColor, texture=/Game/Environments/Helios/Sets/Orchard/Textures/T_Orchard_ModernFrames_A_D}
  NamedRerouteDeclaration(None=@n58.RGB) {name=GlassColor, nodeColor=<rgba 1 0.3047 0 1>}
  n59 = TextureSampleParameter2D() {
   group=GreenHouse
   parameterName=GreenHouseNormal
   samplerType=SAMPLERTYPE_Normal
   texture=/Game/Environments/Helios/Sets/Orchard/Textures/T_Orchard_ModernFrames_A_N
  }
  NamedRerouteDeclaration(None=@n59.RGB) {name=GlassNormal, nodeColor=<rgba 0 0.3906 1 1>}
  n60 = TextureSampleParameter2D() {
   group=GreenHouse
   parameterName=GreenHouse_S
   samplerType=SAMPLERTYPE_Masks
   texture=/Game/Environments/Helios/Sets/Orchard/Textures/T_Orchard_ModernFrames_A_S
  }
  NamedRerouteDeclaration(None=@n60.R) {name=Glass_S, nodeColor=<rgba 1 0.3047 0 1>}
  NamedRerouteDeclaration(None=@n60.G) {name=Glass_M, nodeColor=<rgba 1 0.3047 0 1>}
  NamedRerouteDeclaration(None=@n60.B) {name=Glass_R, nodeColor=<rgba 1 0.3047 0 1>}
  NamedRerouteDeclaration(None=@n3.WorldPositionOffset) {name=WPOIn, nodeColor=<rgba 1 0.5391 0 1>}
  n61 = ObjectPositionWS()
  n62 = WorldPosition()
  NamedRerouteDeclaration(None=ShadowReplace(Default=NamedRerouteUsage(), Shadow=Subtract(A=@n61, B=@n62.XYZ))) {name=WPOout, nodeColor=<rgba 0.6875 0 1 1>}
  NamedRerouteDeclaration(None=Multiply(A=Saturate(None=@n35), B=ScalarParameter() {defaultValue=0.3, group=GreenHouse, parameterName=FogMaskIntensity})) {name=FogMask, nodeColor=<rgba 0.7812 0 1 1>}
  n63 = Constant3Vector() {constant=<rgba 0.8538 0.99 1 1>}
  NamedRerouteDeclaration(None=MaterialFunctionCall(
    Color_Morning=Multiply(A=@n63) {constB=0.5}
    Color_Day=@n63
    Color_Evening=Multiply(A=@n63) {constB=0.3}
    Color_Night=Multiply(A=@n63) {constB=0.1}
   ) {materialFunction=&Color_Time_of_Day}.Result) {name=FogColor, nodeColor=<rgba 0 0.7422 1 1>}
  NamedRerouteDeclaration(None=Multiply(A=@n53, B=VectorParameter() {group=GreenHouse, parameterName=SSSTint, defaultValue=<rgba 1 1 1 0>}.RGB)) {name=SSSColorOUT, nodeColor=<rgba 0.7344 1 0 1>}
  n64 = Multiply(A=@n53, B=VectorParameter() {group=GreenHouse, parameterName=BaseColorTint, defaultValue=<rgba 1 1 1 1>}.RGB)
  n65 = NamedRerouteUsage()
  n66 = Constant() {r=1}
  n67 = Multiply(
   A=Multiply(A=@n64, B=Desaturation(None=@n65, Fraction=@n66))
   B=ScalarParameter() {defaultValue=10, group=GreenHouse, parameterName=BaseColorMulti}
  )
  NamedRerouteDeclaration(None=StaticSwitchParameter(True=Saturate(None=@n67), False=@n64) {group=GreenHouse, parameterName=BlendOrginalBaseColor}) {name=BaseColorOUT, nodeColor=<rgba 1 0.7969 0 1>}
  n68 = ScalarParameter() {group=B_SSS_EdgeDepth, parameterName=SSS Depth Min, sortPriority=4}
  n69 = StaticSwitchParameter(True=TextureCoordinate() {coordinateIndex=2}, False=TextureCoordinate() {coordinateIndex=1}) {group=B_SSS_EdgeDepth, parameterName=SSS Depth UV2, sortPriority=-30}
  n70 = TextureCoordinate()
  n71 = ScalarParameter() {defaultValue=1, group=B_SSS_EdgeDepth, parameterName=SSSDepth X, sortPriority=5}
  n72 = ScalarParameter() {defaultValue=1, group=B_SSS_EdgeDepth, parameterName=SSSDepth Y, sortPriority=5}
  n73 = ComponentMask(None=MaterialFunctionCall!(
    UVs=StaticSwitchParameter(True=@n69, False=@n70) {desc="Default is UV0 ", group=B_SSS_EdgeDepth, parameterName=SSS Depth UV1, sortPriority=-30}
    Tiling=AppendVector(A=@n71, B=@n72)
    Randomize Rotation
   ) {materialFunction=/Engine/Functions/Engine_MaterialFunctions01/Cubemaps/InteriorCubemap}.UVW) {r=true}
  n74 = Constant() {r=0.8}
  n75 = Saturate(None=Divide(A=Add(A=@n73, B=@n74), B=ScalarParameter() {defaultValue=3, group=B_SSS_EdgeDepth, parameterName=SSS_Depth Edge Blur, sortPriority=3}))
  NamedRerouteDeclaration(None=OneMinus(None=LinearInterpolate(A=@n68, Alpha=@n75))) {name=FogBlend, nodeColor=<rgba 0 0.7891 1 1>}
  NamedRerouteDeclaration(None=@n3.BaseColor) {name=BaseColorIn, nodeColor=<rgba 1 0 0.75 1>}
  OneMinus(None=NamedRerouteDeclaration(None=MaterialFunctionCall!(UVs, Mult, Exp) {materialFunction=/Game/Effects/Fort_Effects/MaterialFunctions/MF_SquareEdgeMask}.SquareMask) {name=SquareMask, nodeColor=<rgba 1 0.8203 0 1>})
 ]}
